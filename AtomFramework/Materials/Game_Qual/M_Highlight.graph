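# Two-sided lit highlight material with a tinted, iridescent rim and a pulsing vertical glow, apparently for marking objects.
# Fresnel blends the color parameter toward a hue shift driven by the reflection vector and a time-wobbled camera direction.
# A sine-animated height gradient, scaled by bounds, drives front-face emissive and joins Fresnel in opacity, which also feeds DitherTemporalAA for the mask.
# Normals blend vertex normals with directions radiating from the instance origin; back faces emit a constant glow.
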Material /AtomFramework/Materials/Game_Qual/M_Highlight
graph {
 properties {bUseMaterialAttributes=true, translucencyLightingMode=TLM_VolumetricDirectional, twoSided=true, usedShadingModels=Default Lit}
 nodes [
  Constant() {r=1}
  n1 = Subtract(
   A=WorldPosition().XYZ
   B=TransformPosition(None=Constant3Vector()) {transformSourceType=TRANSFORMPOSSOURCE_Instance, transformType=TRANSFORMPOSSOURCE_World}
  )
  n2 = Normalize(VectorInput=@n1)
  n3 = Sine(None=Time())
  Multiply(A=@n2, B=Add(A=ConstantBiasScale(None=@n3) {scale=0.2}) {constB=0.4})
  ScalarParameter() {defaultValue=12, parameterName=Pulse}
  n4 = Fresnel() {baseReflectFraction=0, exponent=2}
  n5 = VectorParameter() {defaultValue=<rgba 0.1191 0.5839 0.9531 1>}
  n6 = Time()
  n7 = Add(
   A=MaterialFunctionCall() {materialFunction=&CameraDirectionVector}.Result
   B=Multiply(A=Constant3Vector() {constant=<rgba 0.5 -0.4 2.4071 0>}, B=Sine(None=@n6) {period=5})
  )
  n8 = MaterialFunctionCall(Hue Shift Percentage=DotProduct(A=ReflectionVectorWS(), B=Normalize(VectorInput=@n7))) {materialFunction=&HueShift}
  n9 = Fresnel() {baseReflectFraction=0.4, exponent=0.5}
  n10 = Multiply(A=Saturate(None=Power(Base=@n9) {constExponent=4})) {constB=0.3}
  n11 = ComponentMask(None=Bounds().Half Extents) {b=true}
  n12 = Multiply(A=@n11) {constB=2}
  n13 = Divide(A=Add(A=ComponentMask(None=@n1) {b=true}, B=@n11), B=Add(A=@n12, B=Multiply(A=@n3, B=@n12))) {constB=32}
  n14 = Power(Base=OneMinus(None=Saturate(None=@n13)))
  n15 = TwoSidedSign()
  n16 = ScalarParameter() {defaultValue=-0.6864, parameterName=Min Opacity}
  n17 = Add(A=Saturate(None=Fresnel(BaseReflectFractionIn=@n16) {exponent=0.75}), B=@n14)
  n18 = VertexNormalWS()
  SetMaterialAttributes!(
   MaterialAttributes
   Base Color=Add(A=Multiply(A=@n4) {constB=0.2}, B=LinearInterpolate(A=@n5.RGB, B=@n8.Result, Alpha=@n10))
   Emissive Color=LinearInterpolate(B=Multiply(A=@n5.RGB, B=@n14), Alpha=Saturate(None=@n15)) {constA=2}
   Opacity=@n17
   Normal=Transform(None=LinearInterpolate(A=@n2, B=@n18) {constAlpha=0.7}) {transformSourceType=TRANSFORMSOURCE_World, transformType=TRANSFORM_Tangent}
   World Position Offset
   Opacity Mask=MaterialFunctionCall(Alpha Threshold=@n17) {materialFunction=&DitherTemporalAA}.Result
  ) {attributeSetTypes [
    #1
    #2
    #3
    #4
    #5
    #6
   ]}
  Subtract()
  TransformPosition(None=Constant3Vector()) {transformSourceType=TRANSFORMPOSSOURCE_Instance, transformType=TRANSFORMPOSSOURCE_World}
  VertexNormalWS()
  WorldPosition()
 ]
}
